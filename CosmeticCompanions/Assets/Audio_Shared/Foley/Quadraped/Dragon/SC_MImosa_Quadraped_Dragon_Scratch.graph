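# Sound cue for quadruped dragon companion scratch foley, routed to the creature pets sound class.
# A quality-level node sends its first two levels to a random pick of five scratch waves and its last two to a pick of the first four.
# All picks are equally weighted.
# Playback runs at reduced volume with companion foley attenuation and reuses the quadruped footstep concurrency set.
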SoundCue /CosmeticCompanions/Assets/Audio_Shared/Foley/Quadraped/Dragon/SC_MImosa_Quadraped_Dragon_Scratch
root=QualityLevel_0
nodes [
 {id=QualityLevel_0, children [Random_2,Random_2,Random_3,Random_3]}
 {id=Random_2, children [WavePlayer_14, WavePlayer_15, WavePlayer_16, WavePlayer_17, WavePlayer_18], properties {weights [1,1,1,1,1]}}
 {id=Random_3, children [WavePlayer_19, WavePlayer_20, WavePlayer_21, WavePlayer_22], properties {weights [1,1,1,1]}}
 {
  id=WavePlayer_14
  properties {soundWaveAssetPtr=/CosmeticCompanions/Assets/Audio_Shared/Foley/Quadraped/Dragon/Wavs/Mimosa_Quadruped_Dragon_Scratch_01}
 }
 {
  id=WavePlayer_15
  properties {soundWaveAssetPtr=/CosmeticCompanions/Assets/Audio_Shared/Foley/Quadraped/Dragon/Wavs/Mimosa_Quadruped_Dragon_Scratch_02}
 }
 {
  id=WavePlayer_16
  properties {soundWaveAssetPtr=/CosmeticCompanions/Assets/Audio_Shared/Foley/Quadraped/Dragon/Wavs/Mimosa_Quadruped_Dragon_Scratch_03}
 }
 {
  id=WavePlayer_17
  properties {soundWaveAssetPtr=/CosmeticCompanions/Assets/Audio_Shared/Foley/Quadraped/Dragon/Wavs/Mimosa_Quadruped_Dragon_Scratch_04}
 }
 {
  id=WavePlayer_18
  properties {soundWaveAssetPtr=/CosmeticCompanions/Assets/Audio_Shared/Foley/Quadraped/Dragon/Wavs/Mimosa_Quadruped_Dragon_Scratch_05}
 }
 {
  id=WavePlayer_19
  properties {soundWaveAssetPtr=&Mimosa_Quadruped_Dragon_Scratch_01}
 }
 {
  id=WavePlayer_20
  properties {soundWaveAssetPtr=&Mimosa_Quadruped_Dragon_Scratch_02}
 }
 {
  id=WavePlayer_21
  properties {soundWaveAssetPtr=&Mimosa_Quadruped_Dragon_Scratch_03}
 }
 {
  id=WavePlayer_22
  properties {soundWaveAssetPtr=&Mimosa_Quadruped_Dragon_Scratch_04}
 }
]
properties {
 attenuationSettings=/Game/Sounds/Attenuation/AS_Companion_Foley
 maxDistance=1200
 priority=0
 soundClassObject=&Fort_Creatures_Pets
 volumeMultiplier=0.5
 concurrencySet [/CosmeticCompanions/Assets/Audio_Shared/Foley/Quadraped/CONC_Mimosa_Quadraped_Footstep]
}
