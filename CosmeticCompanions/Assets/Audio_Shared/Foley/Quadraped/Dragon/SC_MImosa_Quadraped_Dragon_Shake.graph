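# Sound Cue for the shake foley of a quadruped dragon companion pet in the Mimosa cosmetic companions set.
# A Random node picks one of three equally weighted shake wave variations on each play.
# The cue plays slightly pitched down and louder, routed to the creature pets sound class with shared companion foley attenuation.
# It reuses the quadruped footstep concurrency set to limit overlapping instances.
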SoundCue /CosmeticCompanions/Assets/Audio_Shared/Foley/Quadraped/Dragon/SC_MImosa_Quadraped_Dragon_Shake
root=Random_2
nodes [
 {id=Random_2, children [WavePlayer_0,WavePlayer_1,WavePlayer_2], properties {weights [1,1,1]}}
 {
  id=WavePlayer_0
  properties {soundWaveAssetPtr=/CosmeticCompanions/Assets/Audio_Shared/Foley/Quadraped/Dragon/Wavs/Mimosa_Quadruped_Dragon_Shake_01}
 }
 {
  id=WavePlayer_1
  properties {soundWaveAssetPtr=/CosmeticCompanions/Assets/Audio_Shared/Foley/Quadraped/Dragon/Wavs/Mimosa_Quadruped_Dragon_Shake_02}
 }
 {
  id=WavePlayer_2
  properties {soundWaveAssetPtr=/CosmeticCompanions/Assets/Audio_Shared/Foley/Quadraped/Dragon/Wavs/Mimosa_Quadruped_Dragon_Shake_03}
 }
]
properties {
 attenuationSettings=/Game/Sounds/Attenuation/AS_Companion_Foley
 maxDistance=1200
 pitchMultiplier=0.9
 soundClassObject=&Fort_Creatures_Pets
 volumeMultiplier=1.3
 concurrencySet [/CosmeticCompanions/Assets/Audio_Shared/Foley/Quadraped/CONC_Mimosa_Quadraped_Footstep]
}
