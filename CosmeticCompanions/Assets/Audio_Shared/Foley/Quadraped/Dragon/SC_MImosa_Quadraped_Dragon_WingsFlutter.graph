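# Sound cue for the wing flutter foley of a quadruped dragon cosmetic companion, played as a low-priority pet creature sound.
# A quality-level switch sends the first two tiers to an evenly weighted random pick of six flutter waves and the other two to a pick of only the first four.
# It uses companion foley attenuation and shares the quadruped footstep concurrency limit.
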SoundCue /CosmeticCompanions/Assets/Audio_Shared/Foley/Quadraped/Dragon/SC_MImosa_Quadraped_Dragon_WingsFlutter
root=QualityLevel_0
nodes [
 {id=QualityLevel_0, children [Random_2,Random_2,Random_3,Random_3]}
 {id=Random_2, children [WavePlayer_0,WavePlayer_1,WavePlayer_2,WavePlayer_3,WavePlayer_4,WavePlayer_5], properties {weights [1,1,1,1,1,1]}}
 {id=Random_3, children [WavePlayer_6,WavePlayer_7,WavePlayer_8,WavePlayer_9], properties {weights [1,1,1,1]}}
 {
  id=WavePlayer_0
  properties {soundWaveAssetPtr=/CosmeticCompanions/Assets/Audio_Shared/Foley/Quadraped/Dragon/Wavs/Mimosa_Quadruped_Dragon_Wings_Flutter_01}
 }
 {
  id=WavePlayer_1
  properties {soundWaveAssetPtr=/CosmeticCompanions/Assets/Audio_Shared/Foley/Quadraped/Dragon/Wavs/Mimosa_Quadruped_Dragon_Wings_Flutter_02}
 }
 {
  id=WavePlayer_2
  properties {soundWaveAssetPtr=/CosmeticCompanions/Assets/Audio_Shared/Foley/Quadraped/Dragon/Wavs/Mimosa_Quadruped_Dragon_Wings_Flutter_03}
 }
 {
  id=WavePlayer_3
  properties {soundWaveAssetPtr=/CosmeticCompanions/Assets/Audio_Shared/Foley/Quadraped/Dragon/Wavs/Mimosa_Quadruped_Dragon_Wings_Flutter_04}
 }
 {
  id=WavePlayer_4
  properties {soundWaveAssetPtr=/CosmeticCompanions/Assets/Audio_Shared/Foley/Quadraped/Dragon/Wavs/Mimosa_Quadruped_Dragon_Wings_Flutter_05}
 }
 {
  id=WavePlayer_5
  properties {soundWaveAssetPtr=/CosmeticCompanions/Assets/Audio_Shared/Foley/Quadraped/Dragon/Wavs/Mimosa_Quadruped_Dragon_Wings_Flutter_06}
 }
 {
  id=WavePlayer_6
  properties {soundWaveAssetPtr=&Mimosa_Quadruped_Dragon_Wings_Flutter_01}
 }
 {
  id=WavePlayer_7
  properties {soundWaveAssetPtr=&Mimosa_Quadruped_Dragon_Wings_Flutter_02}
 }
 {
  id=WavePlayer_8
  properties {soundWaveAssetPtr=&Mimosa_Quadruped_Dragon_Wings_Flutter_03}
 }
 {
  id=WavePlayer_9
  properties {soundWaveAssetPtr=&Mimosa_Quadruped_Dragon_Wings_Flutter_04}
 }
]
properties {
 attenuationSettings=/Game/Sounds/Attenuation/AS_Companion_Foley
 maxDistance=1200
 priority=0
 soundClassObject=&Fort_Creatures_Pets
 volumeMultiplier=0.8
 concurrencySet [/CosmeticCompanions/Assets/Audio_Shared/Foley/Quadraped/CONC_Mimosa_Quadraped_Footstep]
}
